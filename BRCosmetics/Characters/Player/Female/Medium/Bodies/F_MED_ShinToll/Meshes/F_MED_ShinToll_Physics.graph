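# Physics asset for the F_MED_ShinToll female player body, setting up physics for the caplet on the character's back.
# Three chains of small capsules (center, left and right), each three bones long, attach to spine_05 through root joints with free swing and twist.
# Joints further down each chain have tight swing and twist limits.
# Larger capsules on spine_04 and spine_05 form the torso shapes behind the chains.
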PhysicsAsset /BRCosmetics/Characters/Player/Female/Medium/Bodies/F_MED_ShinToll/Meshes/F_MED_ShinToll_Physics
bodies {
 dyn_caplet_back_0 [{shapeType=Capsule, length=0.505, radius=0.505, shapeName=dyn_caplet_back_0_capsule, rotation=<rot 0 -90 0>}]
 dyn_caplet_back_1 [{shapeType=Capsule, length=5, radius=0.7, shapeName=dyn_caplet_back_1_capsule, center=<1 -1.5 0>}]
 dyn_caplet_back_2 [{shapeType=Capsule, length=5, radius=1, shapeName=dyn_caplet_back_2_capsule, center=<1 -1.5 0>}]
 dyn_caplet_back_side_l_0 [{shapeType=Capsule, length=0.505, radius=0.505, shapeName=dyn_caplet_back_side_l_0_capsule, rotation=<rot 0 -90 0>}]
 dyn_caplet_back_side_l_1 [{shapeType=Capsule, length=5, radius=0.7, shapeName=dyn_caplet_back_side_l_1_capsule, center=<1 0 0>}]
 dyn_caplet_back_side_l_2 [{shapeType=Capsule, length=5, radius=1, shapeName=dyn_caplet_back_side_l_2_capsule, center=<1 0 0>}]
 dyn_caplet_back_side_r_0 [{shapeType=Capsule, length=0.505, radius=0.505, shapeName=dyn_caplet_back_side_r_0_capsule, rotation=<rot 0 -90 0>}]
 dyn_caplet_back_side_r_1 [{shapeType=Capsule, length=5, radius=0.8, shapeName=dyn_caplet_back_side_r_1_capsule, center=<1 0 0>}]
 dyn_caplet_back_side_r_2 [{shapeType=Capsule, length=5, radius=1, shapeName=dyn_caplet_back_side_r_2_capsule, center=<1 0 0>}]
 spine_04 [{shapeType=Capsule, length=8, radius=8, shapeName=spine_04_capsule, center=<0.5053 5.0986 -2.7533>, rotation=<rot 4.6915 -90.7041 -2.2986>}]
 spine_05 [{shapeType=Capsule, length=10.5, radius=18.6, shapeName=spine_05_capsule, center=<-4.8207 9.8446 -0.1982>, rotation=<rot 90 0 0>}]
}
constraints [
 {bone1Name=dyn_caplet_back_side_r_2, bone2Name=dyn_caplet_back_side_r_1, swing1LimitDegrees=15, swing2LimitDegrees=10, twistLimitDegrees=3}
 {bone1Name=dyn_caplet_back_side_r_1, bone2Name=dyn_caplet_back_side_r_0, swing1LimitDegrees=15, swing2LimitDegrees=5, twistLimitDegrees=0}
 {
  bone1Name=dyn_caplet_back_side_r_0
  bone2Name=spine_05
  swing1LimitDegrees=45
  swing1Motion=Free
  swing2LimitDegrees=45
  swing2Motion=Free
  twistLimitDegrees=45
  twistMotion=Free
 }
 {bone1Name=dyn_caplet_back_side_l_2, bone2Name=dyn_caplet_back_side_l_1, swing1LimitDegrees=15, swing2LimitDegrees=10, twistLimitDegrees=5}
 {bone1Name=dyn_caplet_back_side_l_1, bone2Name=dyn_caplet_back_side_l_0, swing1LimitDegrees=15, swing2LimitDegrees=5, twistLimitDegrees=0}
 {
  bone1Name=dyn_caplet_back_side_l_0
  bone2Name=spine_05
  swing1LimitDegrees=45
  swing1Motion=Free
  swing2LimitDegrees=45
  swing2Motion=Free
  twistLimitDegrees=45
  twistMotion=Free
 }
 {bone1Name=dyn_caplet_back_2, bone2Name=dyn_caplet_back_1, swing1LimitDegrees=15, swing2LimitDegrees=5, twistLimitDegrees=0}
 {bone1Name=dyn_caplet_back_1, bone2Name=dyn_caplet_back_0, swing1LimitDegrees=15, swing2LimitDegrees=5, twistLimitDegrees=0}
 {
  bone1Name=dyn_caplet_back_0
  bone2Name=spine_05
  swing1LimitDegrees=45
  swing1Motion=Free
  swing2LimitDegrees=45
  swing2Motion=Free
  twistLimitDegrees=45
  twistMotion=Free
 }
]
