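# Collar physics asset for the F_MED_ShinToll female cosmetic body, covering a short cable chain attached at the upper spine.
# Small capsules on base_cable_01 and dyn_cable_01 to 03 form the chain, with a larger capsule on spine_04.
# Constraints between cable segments keep one swing axis tight and widen the other toward the tip.
# The joint from base_cable_01 to spine_04 leaves swing and twist free.
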PhysicsAsset /BRCosmetics/Characters/Player/Female/Medium/Bodies/F_MED_ShinToll/Meshes/F_MED_ShinToll_Physics_collar
bodies {
 base_cable_01 [{shapeType=Capsule, length=0.505, radius=0.505, shapeName=base_cable_01_capsule, rotation=<rot 0 -90 0>}]
 dyn_cable_01 [{shapeType=Capsule, length=5, radius=0.7, shapeName=dyn_cable_01_capsule, center=<1 0 0>, rotation=<rot 0 0 90>}]
 dyn_cable_02 [{shapeType=Capsule, length=5, radius=0.7, shapeName=dyn_cable_02_capsule, center=<1 0 0.2>, rotation=<rot 0 0 90>}]
 dyn_cable_03 [{shapeType=Capsule, length=5, radius=0.7, shapeName=dyn_cable_03_capsule, center=<1 0 0>, rotation=<rot 0 0 90>}]
 spine_04 [{shapeType=Capsule, length=8, radius=8, shapeName=spine_04_capsule, center=<0.505 5.0781 -2.7516>, rotation=<rot 4.6915 -90.7041 -2.2986>}]
}
constraints [
 {bone1Name=dyn_cable_03, bone2Name=dyn_cable_02, swing1LimitDegrees=5, swing2LimitDegrees=50, twistLimitDegrees=45}
 {bone1Name=dyn_cable_02, bone2Name=dyn_cable_01, swing1LimitDegrees=5, swing2LimitDegrees=40, twistLimitDegrees=45}
 {bone1Name=dyn_cable_01, bone2Name=base_cable_01, swing1LimitDegrees=5, swing2LimitDegrees=30, twistLimitDegrees=45}
 {
  bone1Name=base_cable_01
  bone2Name=spine_04
  swing1LimitDegrees=45
  swing1Motion=Free
  swing2LimitDegrees=45
  swing2Motion=Free
  twistLimitDegrees=45
  twistMotion=Free
 }
]
