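# Animation Blueprint for the F_MED_ShinToll character body, adding secondary motion to the parent mesh's copied pose.
# Control Rig feeds a weighted right-pocket constraint, pocket AnimDynamics and collar RigidBody simulation, cached for reuse.
# A second spine_05-space RigidBody runs only without backbling, as the blend bypasses it when isWearingBackblingCape is true.
# The event graph updates that flag, computes GravityOverride from a parameter asset, and toggles cape and body morph targets for lower LODs.
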AnimBlueprint /BRCosmetics/Characters/Player/Female/Medium/Bodies/F_MED_ShinToll/Meshes/F_MED_ShinToll_AnimBP
target_skeleton=&/F_MED_ShinToll_Skeleton
graphs [
 AnimGraph {nodes [
   Output Pose(Result=Use cached pose 'predyn_cache'()) {layerGroup=None}
   n1 = RigidBody!(
    ComponentPose=Local To Component(LocalPose=Use cached pose 'control_rig_cache'())
    bAlphaBoolEnabled:bool
    OverrideWorldGravity:Vector="0,0,-980"
   ) -> (Pose) {
    alphaInputType=Bool
    bClampLinearTranslationLimitToRefPose=true
    bOverrideWorldGravity=true
    overridePhysicsAsset=&/F_MED_ShinToll_Physics
    simulationSpace=BaseBoneSpace
    alphaBoolBlend {blendOption=Linear}
    baseBoneRef {boneName=spine_05}
    overrideWorldGravity {z=-980}
    simSpaceSettings {
     dampingAlpha=0
     maxAngularAcceleration=9
     maxAngularVelocity=4
     maxLinearAcceleration=1000
     maxLinearVelocity=400
     worldAlpha=1
     externalLinearDragV=<1.1 1.1 1.1>
    }
   }
   Save cached pose 'predyn_cache'(Pose=Blend Poses by bool(
     BlendPose_0=Use cached pose 'control_rig_cache'()
     BlendPose_1=Component To Local(ComponentPose=@n1)
     bActiveValue=Get isWearingBackblingCape() -> (isWearingBackblingCape:bool)
    )) {cacheName=predyn_cache}
   n2 = Control Rig(Source=Copy Pose From Mesh() {bUseAttachedParent=true}) {controlRigAssetReference {blueprintRigClass=&/F_MED_ShinToll_CtrlRig.F_MED_ShinToll_CtrlRig_C}}
   n3 = "Constraint - base_pocket_attach_down_r (PPPPP)"!(
    ComponentPose=Local To Component(LocalPose=@n2)
    Alpha=1
    ConstraintWeights_0=0.786
    ConstraintWeights_1:float=0.2104
    ConstraintWeights_2:float=0.0025
    ConstraintWeights_3:float=0.0009
    ConstraintWeights_4:float=0.0002
   ) -> (Pose) {
    boneToModify {boneName=base_pocket_attach_down_r}
    constraintSetup [
     {targetBone {boneName=thigh_r}}
     {targetBone {boneName=thigh_twist_01_r}}
     {targetBone {boneName=pelvis}}
     {targetBone {boneName=deform_groin_r}}
     {targetBone {boneName=deform_knee_out_r}}
    ]
    constraintWeights [1,1,1,1,1]
   }
   n4 = Anim Dynamics - Bone: pocket_attach_down_r(ComponentPose=@n3, Alpha=1, ExternalForce="0,0,0") {
    angularDampingOverride=1
    angularSpringConstant=150
    bAngularSpring=true
    bOverrideAngularDamping=true
    bUseGravityOverride=true
    gravityScale=0.35
    simulationSpace=RootRelative
    boundBone {boneName=pocket_attach_down_r}
    componentLinearAccScale=<1 1 1>
    componentLinearVelScale=<1 1 1>
    gravityOverride {z=-980}
    physicsBodyDefinitions [{
      boundBone {boneName=pocket_attach_down_r}
      boxExtents=<10 10 10>
      constraintSetup {angularLimitsMax=<3 3 5>, angularLimitsMin=<-5 -3 -5>, angularTarget=<1 0 0>}
      localJointOffset=<2 0 0>
     }]
   }
   n5 = RigidBody!(ComponentPose=@n4, OverrideWorldGravity="0,0,-980") -> (Pose) {
    bClampLinearTranslationLimitToRefPose=true
    bOverrideWorldGravity=true
    overridePhysicsAsset=&/F_MED_ShinToll_Physics_collar
    simulationSpace=BaseBoneSpace
    alphaBoolBlend {blendOption=Linear}
    baseBoneRef {boneName=spine_05}
    overrideWorldGravity {z=-980}
    simSpaceSettings {
     dampingAlpha=0
     maxAngularAcceleration=200
     maxAngularVelocity=200
     maxLinearAcceleration=1000
     maxLinearVelocity=300
     worldAlpha=1
     externalLinearDragV=<0.3 0.3 0.3>
    }
   }
   Save cached pose 'control_rig_cache'(Pose=Component To Local(ComponentPose=@n5)) {cacheName=control_rig_cache}
  ]}
 EventGraph {nodes [
   Event BlueprintUpdateAnimation()
   > n1 = Sequence!(execute) -> (then_0, then_1, then_2:Exec)
   > n2 = "Gravity Override State Function (Parameter Asset) (Nativized)"(
    CharacterPartAnimInstance=Self-Reference()
    GameTimeInSeconds=GetGameTimeInSeconds()
    Parameters=&/F_MED_ShinToll_Parameters
   )
   > Set GravityOverride(GravityOverride=@n2)
   n3 = CheckForBackbling()
   n4 = NOT Boolean(A=@n3.bIsBackblingLong)
   n5 = NOT Boolean(A=@n3.bIsBackblingACape)
   n6 = Branch(
    execute=@n1.then_1
    Condition=OR Boolean(A=AND Boolean!(A=@n3, B=@n5, C:bool=@n4) -> (ReturnValue), B=OR Boolean(A=@n3.bIsBackblingACape, B=@n3.bIsBackblingLong))
   )
   > Set isWearingBackblingCape(isWearingBackblingCape:bool=true) -> (Output_Get:bool)
   Set isWearingBackblingCape(execute=@n6.else, isWearingBackblingCape=false)
   n7 = Branch(execute=@n1.then_2, Condition="integer > integer"(A=Get CurrentLOD(), B=0) {is=title K2Node_CallFunction})
   > SetMorphTarget!(execute, self:Anim Instance Object Reference, MorphTargetName:Name=cape_LODs_Morph, Value:float=1) -> (then)
   > SetMorphTarget(MorphTargetName=body_LODs_Morph, Value=1)
   SetMorphTarget(execute=@n7.else, MorphTargetName=cape_LODs_Morph, Value=0)
   > SetMorphTarget(MorphTargetName=body_LODs_Morph, Value=0)
  ]}
]
variables [
 GravityOverride {type=Vector}
 isWearingBackblingCape {type=boolean}
]
