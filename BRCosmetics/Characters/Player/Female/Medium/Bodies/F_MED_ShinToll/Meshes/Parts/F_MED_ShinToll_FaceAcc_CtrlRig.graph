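# Control Rig for the ShinToll female character's face accessory part that applies a pose-driven corrective offset based on upper arm orientation.
# Two spherical pose readers track the left and right upperarm bones with mirrored axes and rotation offsets.
# Their halved outputs are summed into one weight for a Modify Transforms step, which adjusts spine_05_helper, upperarm_l and further items cut off in the listing.
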ControlRigBlueprint /BRCosmetics/Characters/Player/Female/Medium/Bodies/F_MED_ShinToll/Meshes/Parts/F_MED_ShinToll_FaceAcc_CtrlRig
graphs [
 RigVMFunctionLibrary {links [], nodes []}
 RigVMModel {nodes [
   RigUnit_BeginExecution = &RigUnit_BeginExecution()
   n1 = &RigUnit_SphericalPoseReader(
    ActiveRegionScaleFactors=(PositiveWidth=2,NegativeWidth=2,PositiveHeight=1,NegativeHeight=1)
    ActiveRegionSize=0.05
    Debug=(bDrawDebug=false,bDraw2D=False,bDrawLocalAxes=False,DebugScale=25,DebugSegments=20,DebugThickness=0.25)
    DriverAxis=(X=1,Y=0,Z=0)
    DriverItem=(Type=Bone,Name="upperarm_l")
    FalloffRegionScaleFactors=(PositiveWidth=1,NegativeWidth=1,PositiveHeight=0.75,NegativeHeight=0.75)
    FalloffSize=0.2
    FlipHeightScaling=False
    FlipWidthScaling=False
    OptionalParentItem=(Type=Bone,Name="None")
    RotationOffset=(X=-15,Y=-30,Z=0)
    ExecutePin=@RigUnit_BeginExecution.ExecutePin
   )
   > n2 = &RigUnit_SphericalPoseReader(
    ActiveRegionScaleFactors=(PositiveWidth=2,NegativeWidth=2,PositiveHeight=1,NegativeHeight=1)
    ActiveRegionSize=0.05
    Debug=(bDrawDebug=false,bDraw2D=False,bDrawLocalAxes=False,DebugScale=25,DebugSegments=20,DebugThickness=0.25)
    DriverAxis=(X=-1,Y=0,Z=0)
    DriverItem=(Type=Bone,Name="upperarm_r")
    FalloffRegionScaleFactors=(PositiveWidth=1,NegativeWidth=1,PositiveHeight=0.75,NegativeHeight=0.75)
    FalloffSize=0.2
    FlipHeightScaling=False
    FlipWidthScaling=False
    OptionalParentItem=(Type=Bone,Name="None")
    RotationOffset=(X=15,Y=150,Z=0)
   )
   Divide = &RigVMFunction_MathDoubleDiv(B=2, A=@n1.OutputParam)
   Add = &RigVMFunction_MathDoubleAdd(A=@Divide.Result, B=&RigVMFunction_MathDoubleDiv(B=2, A=@n2.OutputParam).Result)
   > &RigUnit_ModifyTransforms(
    ItemToModify="((Item=(Type=Bone,Name=\"spine_05_helper\"),Transform=(Rotation=(X=0,Y=0,Z=0,W=1),Translation=(X=10,Y=-5,Z=0),Scale3D=(X=1,Y=1,Z=1))),(Item=(Type=Bone,Name=\"upperarm_l\"),Transform=(Rotation=(X=0,Y=0,Z=0,W=1),Translation=(X=0,Y=0,Z=10),Scale3D=(X=1,Y=1,Z=1))) ...(truncated)"
    WeightMinimum=0
    Weight=@Add.Result
   )
  ]}
]
hierarchy [
 root {loc=<0 0 0>, rot=<0 0 0 1>}
 pelvis {parent=root, loc=<0 -0.5301 98.0571>, rot=<0.0127 0.707 -0.0127 -0.707>}
 spine_01 {parent=pelvis, loc=<6.1208 -0.9328 0>, rot=<0 0 0.034 -0.9994>}
 spine_02 {parent=spine_01, loc=<6.5997 -0.2095 0>, rot=<0 0 0.0238 -0.9997>}
 spine_03 {parent=spine_02, loc=<6.4914 -0.1237 0>, rot=<0 0 0.0149 -0.9999>}
 spine_04 {parent=spine_03, loc=<6.9598 0.0794 0>, rot=<0 0 -0.0132 -0.9999>}
 spine_05 {parent=spine_04, loc=<6.9987 0 0>, rot=<0 0 -0.0015 -1>}
 neck_01 {parent=spine_05, loc=<11.3138 0.1529 0>, rot=<0 0 -0.2003 -0.9797>}
 neck_02 {parent=neck_01, loc=<4.5627 0 0>, rot=<0 0 0.0504 -0.9987>}
 head {parent=neck_02, loc=<5.3914 0 0>, rot=<0 0 0.115 -0.9934>}
 base_hair_side_01_l {parent=head, loc=<4.2451 2.1672 -8.582>, rot=<0.0112 -0.707 -0.7042 0.0637>}
 dyn_hair_side_01_l {parent=base_hair_side_01_l, loc=<0 0 0>, rot=<-0.0047 0.0003 0.0477 -0.9988>}
 dyn_hair_side_02_l {parent=dyn_hair_side_01_l, loc=<2.8255 0 0>, rot=<-0.1473 -0.0951 0.1256 -0.9765>}
 dyn_hair_side_03_l {parent=dyn_hair_side_02_l, loc=<2.7402 0.1045 -0.0242>, rot=<0.0945 -0.1506 0.3162 -0.9319>}
 base_hair_front_01_l {parent=head, loc=<3.8991 5.101 -7.6936>, rot=<0.0554 -0.7049 0.7063 0.0341>}
 dyn_hair_front_01_l {parent=base_hair_front_01_l, loc=<0 0 0>, rot=<0 0 0 1>}
 dyn_hair_front_02_l {parent=dyn_hair_front_01_l, loc=<4.6991 0 0>, rot=<0.0055 -0.1111 -0.0764 -0.9908>}
 dyn_hair_front_03_l {parent=dyn_hair_front_02_l, loc=<4.2147 0 0>, rot=<0.9951 0.0002 0.0987 0.0046>}
 base_hair_side_01_r {parent=head, loc=<1.4858 2.2696 8.1403>, rot=<0.0793 -0.7026 -0.671 0.2232>}
 dyn_hair_side_01_r {parent=base_hair_side_01_r, loc=<0 0 0>, rot=<0 0 0 1>}
 dyn_hair_side_02_r {parent=dyn_hair_side_01_r, loc=<2.4139 0 0>, rot=<0 0 -0.2083 -0.9781>}
 base_hair_front_01_r {parent=head, loc=<3.1067 4.9186 7.5946>, rot=<-0.0584 0.702 0.709 0.0322>}
 dyn_hair_front_01_r {parent=base_hair_front_01_r, loc=<0 0 0>, rot=<0 0 0 1>}
 dyn_hair_front_02_r {parent=dyn_hair_front_01_r, loc=<4.0543 -0.2303 0.1997>, rot=<-0.0074 0.0474 -0.0776 -0.9958>}
 dyn_hair_front_03_r {parent=dyn_hair_front_02_r, loc=<3.7309 0.1355 0.6584>, rot=<0 0.001 0 -1>}
 base_ponytail {parent=head, loc=<0 0 0>, rot=<0 0 0 1>}
 base_ponytail_00_r {parent=base_ponytail, loc=<12.0694 -5.0129 8.3469>, rot=<0.7955 -0.139 0.5011 -0.311>}
 dyn_ponytail_00_r {parent=base_ponytail_00_r, loc=<0 0 0>, rot=<0 0 0 1>}
 dyn_ponytail_01_r {parent=dyn_ponytail_00_r, loc=<4.4553 -0.0501 0>, rot=<0.5034 0.4991 -0.5009 0.4966>}
 dyn_ponytail_02_r {parent=dyn_ponytail_01_r, loc=<9.4862 0.2889 0>, rot=<0.8938 -0.438 0.0562 0.0783>}
 dyn_ponytail_03_r {parent=dyn_ponytail_02_r, loc=<10.254 0 0>, rot=<0 0 0.3997 -0.9167>}
 base_ponytail_00_l {parent=base_ponytail, loc=<12.0694 -5.0129 -8.3469>, rot=<0.2622 -0.4904 0.1667 -0.8142>}
 dyn_ponytail_00_l {parent=base_ponytail_00_l, loc=<0 0 0>, rot=<0 0 0 1>}
 dyn_ponytail_01_l {parent=dyn_ponytail_00_l, loc=<4.4553 0.0314 0>, rot=<-0.5002 -0.5062 0.5028 -0.4907>}
 dyn_ponytail_02_l {parent=dyn_ponytail_01_l, loc=<9.4776 -0.4963 0>, rot=<-0.0263 0.0378 0.4217 -0.9055>}
 dyn_ponytail_03_l {parent=dyn_ponytail_02_l, loc=<10.254 0 0>, rot=<0 0 -0.5078 -0.8615>}
 base_hair_bang_front {parent=head, loc=<14.9649 7.0373 0.3911>, rot=<0.0041 -1 0 0>}
 dyn_hair_bang_front {parent=base_hair_bang_front, loc=<0 0 0>, rot=<0 0 0 1>}
 clavicle_l {parent=spine_05, loc=<5.4831 -0.1053 -3.2946>, rot=<0.0465 -0.7289 -0.009 -0.6829>}
 upperarm_l {parent=clavicle_l, loc=<9.8658 -1.2136 -0.4105>, rot=<0.0164 -0.3245 -0.0269 -0.9454>}
 clavicle_r {parent=spine_05, loc=<5.4831 -0.1053 3.2946>, rot=<0.7289 0.0465 -0.6829 0.009>}
 upperarm_r {parent=clavicle_r, loc=<-9.8658 1.2136 0.4105>, rot=<0.0164 -0.3245 -0.0269 -0.9454>}
 spine_05_helper {parent=spine_05, loc=<0 0 0>, rot=<0 0 0 1>}
]
