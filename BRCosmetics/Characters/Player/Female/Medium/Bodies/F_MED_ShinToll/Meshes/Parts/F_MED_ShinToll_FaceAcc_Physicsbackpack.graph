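# Physics asset for the F_MED_ShinToll female character's face accessory, named as a backpack variant, covering front and side hair strands and twin ponytails.
# Base bones anchor chains of dyn_ capsules with tight swing limits and mostly zero or locked twist.
# Head, clavicle, upper-arm and large offset spine_05 capsules provide body volumes.
# Most base bones link freely to spine_05, but the right ponytail base locks to the head.
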PhysicsAsset /BRCosmetics/Characters/Player/Female/Medium/Bodies/F_MED_ShinToll/Meshes/Parts/F_MED_ShinToll_FaceAcc_Physicsbackpack
bodies {
 base_hair_front_01_l [{shapeType=Capsule, length=0.505, radius=0.505, shapeName=base_hair_front_01_l_capsule, rotation=<rot 0 0 -180>}]
 base_hair_front_01_r [{shapeType=Capsule, length=0.505, radius=0.505, shapeName=base_hair_front_01_r_capsule}]
 base_hair_side_01_l [{shapeType=Capsule, length=0.505, radius=0.505, shapeName=base_hair_side_01_l_capsule}]
 base_hair_side_01_r [{shapeType=Capsule, length=0.505, radius=0.505, shapeName=base_hair_side_01_r_capsule}]
 base_ponytail_00_l [{shapeType=Capsule, length=2, radius=1, shapeName=base_ponytail_00_l_capsule, rotation=<rot 0 -90 0>}]
 base_ponytail_00_r [{shapeType=Capsule, length=2, radius=2, shapeName=dyn_ponytail_00_r_capsule, center=<2.2827 0 0>, rotation=<rot 90 0 0>}]
 clavicle_l [{shapeType=Capsule, length=4, radius=7.5, shapeName=clavicle_l_capsule, center=<5.2127 -1.0196 -2.1408>, rotation=<rot 0 -90 -76.0692>}]
 clavicle_r [{shapeType=Capsule, length=4, radius=7.5, shapeName=clavicle_r_capsule, center=<-5.2127 1.0196 2.1408>, rotation=<rot 0 -90 103.9308>}]
 dyn_hair_front_01_l [{shapeType=Capsule, length=2, radius=1.5, shapeName=dyn_hair_front_02_l_capsule, center=<2.3776 0 0>, rotation=<rot 90 0 0>}]
 dyn_hair_front_01_r [{shapeType=Capsule, length=2, radius=1.5, shapeName=dyn_hair_front_02_r_capsule, center=<2.417 0 0>, rotation=<rot 90 0 0>}]
 dyn_hair_front_02_l [{shapeType=Capsule, length=2, radius=1.5, shapeName=dyn_hair_front_02_l_capsule, center=<1.9051 0 0>, rotation=<rot 90 0 0>}]
 dyn_hair_front_02_r [{shapeType=Capsule, length=2, radius=1.5, shapeName=dyn_hair_front_02_r_capsule, center=<1.4657 0 0>, rotation=<rot 90 0 0>}]
 dyn_hair_front_03_l [{shapeType=Capsule, length=2, radius=1, shapeName=dyn_hair_front_03_l_capsule, center=<1.6274 0 0.5>, rotation=<rot -70.762 0 0>}]
 dyn_hair_front_03_r [{shapeType=Capsule, length=2, radius=1.5, shapeName=dyn_hair_front_03_r_capsule, center=<1.2927 0 0.7>, rotation=<rot 120 0 0>}]
 dyn_hair_side_01_l [{shapeType=Capsule, length=1.5, radius=1.2, shapeName=dyn_hair_side_01_l_capsule, center=<0.8949 0 0>, rotation=<rot 90 0 0>}]
 dyn_hair_side_01_r [{shapeType=Capsule, length=1.5, radius=1.2, shapeName=dyn_hair_side_01_r_capsule, center=<0.8 0 0>, rotation=<rot 90 0 0>}]
 dyn_hair_side_02_l [{shapeType=Capsule, length=2, radius=0.7, shapeName=dyn_hair_side_02_l_capsule, center=<1.0406 0 0>, rotation=<rot 90 0 0>}]
 dyn_hair_side_02_r [{shapeType=Capsule, length=2, radius=0.7, shapeName=dyn_hair_side_02_r_capsule, center=<1.3 0 0>, rotation=<rot 90 0 0>}]
 dyn_hair_side_03_l [{shapeType=Capsule, length=2, radius=0.7, shapeName=dyn_hair_side_03_l_capsule, rotation=<rot 90 0 0>}]
 dyn_ponytail_00_l [{shapeType=Capsule, length=2, radius=2, shapeName=dyn_ponytail_00_l_capsule, center=<2.2827 0 0>, rotation=<rot 90 180 0>}]
 dyn_ponytail_00_r [{shapeType=Capsule, length=2, radius=2, shapeName=dyn_ponytail_00_r_capsule, center=<2.2827 0 0>, rotation=<rot 90 0 0>}]
 dyn_ponytail_01_l [{shapeType=Capsule, length=3, radius=4, shapeName=dyn_ponytail_01_l_capsule, center=<4.6484 0 0>, rotation=<rot 90 180 0>}]
 dyn_ponytail_01_r [{shapeType=Capsule, length=3, radius=4, shapeName=dyn_ponytail_01_r_capsule, center=<4.6484 0 0>, rotation=<rot 90 0 0>}]
 dyn_ponytail_02_l [{shapeType=Capsule, length=4.2695, radius=3.5515, shapeName=dyn_ponytail_02_l_capsule, center=<5.5587 0 0>, rotation=<rot 90 180 0>}]
 dyn_ponytail_02_r [{shapeType=Capsule, length=4.2695, radius=3.5515, shapeName=dyn_ponytail_02_r_capsule, center=<5.5587 0 0>, rotation=<rot 90 0 0>}]
 dyn_ponytail_03_l [{shapeType=Capsule, length=6.4, radius=3, shapeName=dyn_ponytail_03_l_capsule, center=<1.3707 1 0>, rotation=<rot -85.6312 0.0003 -172.243>}]
 dyn_ponytail_03_r [{shapeType=Capsule, length=6.4, radius=3, shapeName=dyn_ponytail_03_r_capsule, center=<1.5088 0 0>, rotation=<rot -85.6311 -0.0005 -32.6489>}]
 head [{shapeType=Capsule, length=0, radius=10.2879, center=<8.2381 -0.3466 0.0713>, rotation=<rot -0.5315 -90 -0.2732>}]
 spine_05 [{shapeType=Capsule, length=0.505, radius=16.5, shapeName=spine_05_capsule, center=<-5.288 -14.4733 0>, rotation=<rot 0 -90 0>}]
 upperarm_l [{
   shapeType=Capsule
   length=25
   radius=5.5
   shapeName=upperarm_l_capsule
   center=<12.5705 -1.155 -1.3504>
   rotation=<rot -89.4099 179.983 -179.9828>
  }]
 upperarm_r [{
   shapeType=Capsule
   length=25
   radius=5.5
   shapeName=upperarm_r_capsule
   center=<-12.5705 1.155 1.3504>
   rotation=<rot -89.4099 179.983 0.0172>
  }]
}
constraints [
 {bone1Name=dyn_hair_front_01_r, bone2Name=base_hair_front_01_r, swing1LimitDegrees=10, swing2LimitDegrees=10, twistLimitDegrees=0}
 {bone1Name=dyn_hair_front_02_r, bone2Name=dyn_hair_front_01_r, swing1LimitDegrees=10, swing2LimitDegrees=20, twistLimitDegrees=0}
 {bone1Name=dyn_hair_front_03_r, bone2Name=dyn_hair_front_02_r, swing1LimitDegrees=10, swing2LimitDegrees=20, twistLimitDegrees=0}
 {bone1Name=dyn_hair_side_01_l, bone2Name=base_hair_side_01_l, swing1LimitDegrees=5, swing2LimitDegrees=20, twistLimitDegrees=0}
 {bone1Name=dyn_hair_side_02_l, bone2Name=dyn_hair_side_01_l, swing1LimitDegrees=8, swing2LimitDegrees=10, twistLimitDegrees=0}
 {bone1Name=dyn_hair_side_03_l, bone2Name=dyn_hair_side_02_l, swing1LimitDegrees=15, swing2LimitDegrees=20, twistLimitDegrees=0}
 {bone1Name=dyn_hair_side_01_r, bone2Name=base_hair_side_01_r, swing1LimitDegrees=5, swing2LimitDegrees=20, twistLimitDegrees=0}
 {bone1Name=dyn_hair_side_02_r, bone2Name=dyn_hair_side_01_r, swing1LimitDegrees=8, swing2LimitDegrees=10, twistLimitDegrees=0}
 {bone1Name=dyn_ponytail_03_r, bone2Name=dyn_ponytail_02_r, swing1LimitDegrees=15, swing2LimitDegrees=20, twistLimitDegrees=0}
 {bone1Name=dyn_ponytail_02_r, bone2Name=dyn_ponytail_01_r, swing1LimitDegrees=20, swing2LimitDegrees=20, twistLimitDegrees=0}
 {bone1Name=dyn_ponytail_01_r, bone2Name=dyn_ponytail_00_r, swing1LimitDegrees=20, swing2LimitDegrees=25, twistLimitDegrees=0}
 {bone1Name=dyn_ponytail_03_l, bone2Name=dyn_ponytail_02_l, swing1LimitDegrees=15, swing2LimitDegrees=20, twistLimitDegrees=0}
 {bone1Name=dyn_ponytail_02_l, bone2Name=dyn_ponytail_01_l, swing1LimitDegrees=20, swing2LimitDegrees=20, twistLimitDegrees=0}
 {bone1Name=dyn_ponytail_01_l, bone2Name=dyn_ponytail_00_l, swing1LimitDegrees=20, swing2LimitDegrees=25, twistLimitDegrees=0}
 {bone1Name=dyn_hair_front_03_l, bone2Name=dyn_hair_front_02_l, swing1LimitDegrees=10, swing2LimitDegrees=20, twistLimitDegrees=0}
 {bone1Name=dyn_hair_front_01_l, bone2Name=base_hair_front_01_l, swing1LimitDegrees=10, swing2LimitDegrees=10, twistLimitDegrees=0}
 {bone1Name=dyn_hair_front_02_l, bone2Name=dyn_hair_front_01_l, swing1LimitDegrees=10, swing2LimitDegrees=20, twistLimitDegrees=0}
 {
  bone1Name=upperarm_r
  bone2Name=clavicle_r
  swing1LimitDegrees=45
  swing1Motion=Locked
  swing2LimitDegrees=45
  swing2Motion=Locked
  twistLimitDegrees=45
  twistMotion=Locked
 }
 {
  bone1Name=upperarm_l
  bone2Name=clavicle_l
  swing1LimitDegrees=45
  swing1Motion=Locked
  swing2LimitDegrees=45
  swing2Motion=Locked
  twistLimitDegrees=45
  twistMotion=Locked
 }
 {
  bone1Name=dyn_ponytail_00_r
  bone2Name=base_ponytail_00_r
  swing1LimitDegrees=5
  swing2LimitDegrees=5
  twistLimitDegrees=45
  twistMotion=Locked
 }
 {
  bone1Name=dyn_ponytail_00_l
  bone2Name=base_ponytail_00_l
  swing1LimitDegrees=5
  swing2LimitDegrees=5
  twistLimitDegrees=45
  twistMotion=Locked
 }
 {
  bone1Name=base_ponytail_00_r
  bone2Name=head
  swing1LimitDegrees=45
  swing1Motion=Locked
  swing2LimitDegrees=45
  swing2Motion=Locked
  twistLimitDegrees=45
  twistMotion=Locked
 }
 {
  bone1Name=head
  bone2Name=spine_05
  swing1LimitDegrees=45
  swing1Motion=Free
  swing2LimitDegrees=45
  swing2Motion=Free
  twistLimitDegrees=45
  twistMotion=Free
 }
 {
  bone1Name=base_hair_side_01_l
  bone2Name=spine_05
  swing1LimitDegrees=45
  swing1Motion=Free
  swing2LimitDegrees=45
  swing2Motion=Free
  twistLimitDegrees=45
  twistMotion=Free
 }
 {
  bone1Name=base_hair_front_01_l
  bone2Name=spine_05
  swing1LimitDegrees=45
  swing1Motion=Free
  swing2LimitDegrees=45
  swing2Motion=Free
  twistLimitDegrees=45
  twistMotion=Free
 }
 {
  bone1Name=base_ponytail_00_l
  bone2Name=spine_05
  swing1LimitDegrees=45
  swing1Motion=Free
  swing2LimitDegrees=45
  swing2Motion=Free
  twistLimitDegrees=45
  twistMotion=Free
 }
 {
  bone1Name=base_hair_front_01_r
  bone2Name=spine_05
  swing1LimitDegrees=45
  swing1Motion=Free
  swing2LimitDegrees=45
  swing2Motion=Free
  twistLimitDegrees=45
  twistMotion=Free
 }
 {
  bone1Name=base_hair_side_01_r
  bone2Name=spine_05
  swing1LimitDegrees=45
  swing1Motion=Free
  swing2LimitDegrees=45
  swing2Motion=Free
  twistLimitDegrees=45
  twistMotion=Free
 }
 {
  bone1Name=clavicle_l
  bone2Name=spine_05
  swing1LimitDegrees=45
  swing1Motion=Free
  swing2LimitDegrees=45
  swing2Motion=Free
  twistLimitDegrees=45
  twistMotion=Free
 }
 {
  bone1Name=clavicle_r
  bone2Name=spine_05
  swing1LimitDegrees=45
  swing1Motion=Free
  swing2LimitDegrees=45
  swing2Motion=Free
  twistLimitDegrees=45
  twistMotion=Free
 }
]
